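# Sound cue for Season 22 narrative quests that plays distorted dialog SFX, the D variant, from a Paradigm voice tuning-puzzle wave labeled 75 percent.
# A single looping wave player feeds the output directly, with boosted volume, the Paradigm conduit vocal attenuation and the Fort quests sound class.
# A concurrency override allows one instance at a time and blocks new plays while it is active.
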SoundCue /Game/Athena/Sounds/Quests/S22_Narrative/Quest_Dialog_Distorted_SFX_D_Cue
root=WavePlayer_0
nodes [{id=WavePlayer_0, properties {bLooping=true, soundWaveAssetPtr=&/ParadigmVO_TuningPuzzle_75Percent}}]
properties {
 attenuationSettings=&/Att_Paradigm_Conduit_Vocal
 bOverrideConcurrency=true
 maxDistance=2200
 soundClassObject=&Fort_Quests
 volumeMultiplier=2
 concurrencyOverrides {maxCount=1, resolutionRule=PreventNew}
}
