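# Sound Cue for a Season 22 narrative quest that plays looping Paradigm reality dialog FX as a spatialized ambience.
# Each loop waits a random one to few seconds, then plays one of four equally weighted waves through a Modulator for variation.
# It uses Paradigm conduit vocal attenuation and the quest sound class.
# A concurrency override allows a single instance and blocks new ones while it plays.
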SoundCue /Game/Athena/Sounds/Quests/S22_Narrative/Quest_Dialog_ParadigmRoot_Cue
root=Looping_0
nodes [
 {id=Delay_0, children [Modulator_0], properties {delayMax=3.5, delayMin=1}}
 {id=Looping_0, children [Delay_0]}
 {id=Modulator_0, children [Random_0], properties {volumeMax=1.1}}
 {id=Random_0, children [WavePlayer_0,WavePlayer_1,WavePlayer_2,WavePlayer_3], properties {weights [1,1,1,1]}}
 {id=WavePlayer_0, properties {soundWaveAssetPtr=&/AMB_DX_Reality_Paradigm_FX_01}}
 {id=WavePlayer_1, properties {soundWaveAssetPtr=&/AMB_DX_Reality_Paradigm_FX_02}}
 {id=WavePlayer_2, properties {soundWaveAssetPtr=&/AMB_DX_Reality_Paradigm_FX_03}}
 {id=WavePlayer_3, properties {soundWaveAssetPtr=&/AMB_DX_Reality_Paradigm_FX_04}}
]
properties {
 attenuationSettings=&/Att_Paradigm_Conduit_Vocal
 bOverrideConcurrency=true
 maxDistance=2200
 soundClassObject=&Fort_Quests
 volumeMultiplier=1.3
 concurrencyOverrides {maxCount=1, resolutionRule=PreventNew}
}
