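# Looping sound cue for transmission distortion in the Season 22 narrative quests.
# A single wave player loops the Paradigm computer static wave straight to the output with a slight volume boost.
# It plays in the Fort quests sound class and borrows the Paradigm conduit vocal attenuation.
# A concurrency override allows only one instance at a time and blocks new ones while it plays.
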SoundCue /Game/Athena/Sounds/Quests/S22_Narrative/Quest_Transmission_Distortion_LP_Cue
root=WavePlayer_1
nodes [{id=WavePlayer_1, properties {bLooping=true, soundWaveAssetPtr=&/Paradigm_ComputerStatic_LP}}]
properties {
 attenuationSettings=&/Att_Paradigm_Conduit_Vocal
 bOverrideConcurrency=true
 maxDistance=2200
 soundClassObject=&Fort_Quests
 volumeMultiplier=1.2
 concurrencyOverrides {maxCount=1, resolutionRule=PreventNew}
}
